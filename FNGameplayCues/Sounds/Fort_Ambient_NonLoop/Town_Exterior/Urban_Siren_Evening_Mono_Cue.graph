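# Sound cue for a non-looping evening urban siren, a one-shot ambient sound for town exteriors.
# A quality-level node feeds two branches: the first picks randomly from six siren waves, the second from just two, and the remaining two slots are empty.
# Each branch applies random pitch variation through a modulator.
# It uses tile-ambient one-shot attenuation and overrides concurrency to prevent new instances at the limit.
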SoundCue /FNGameplayCues/Sounds/Fort_Ambient_NonLoop/Town_Exterior/Urban_Siren_Evening_Mono_Cue
root=QualityLevel_0
nodes [
 {id=Modulator_1, children [Random_0], properties {pitchMax=1.3, pitchMin=0.9}}
 {id=Modulator_2, children [Random_1], properties {pitchMax=1.3, pitchMin=0.9}}
 {id=QualityLevel_0, children [Modulator_1,Modulator_2,null,null]}
 {id=Random_0, children [WavePlayer_8,WavePlayer_1,WavePlayer_2,WavePlayer_3,WavePlayer_4,WavePlayer_5], properties {weights [1,1,1,1,1,1]}}
 {id=Random_1, children [WavePlayer_9,WavePlayer_6], properties {weights [1,1]}}
 {id=WavePlayer_1, properties {soundWaveAssetPtr=&/urban_siren_evening_02}}
 {id=WavePlayer_2, properties {soundWaveAssetPtr=&/urban_siren_evening_03}}
 {id=WavePlayer_3, properties {soundWaveAssetPtr=&/urban_siren_evening_04}}
 {id=WavePlayer_4, properties {soundWaveAssetPtr=&/urban_siren_evening_05}}
 {id=WavePlayer_5, properties {soundWaveAssetPtr=&/urban_siren_evening_06}}
 {id=WavePlayer_6, properties {soundWaveAssetPtr=&urban_siren_evening_02}}
 {id=WavePlayer_8, properties {soundWaveAssetPtr=&/urban_siren_evening_01}}
 {id=WavePlayer_9, properties {soundWaveAssetPtr=&urban_siren_evening_01}}
]
properties {
 attenuationSettings=/Game/Sounds/Attenuation/Fort_TileAmbient_OneShot_Att
 bOverrideConcurrency=true
 maxDistance=8000
 soundClassObject=/Game/Sounds/FortSoundClassesAndModes/Fort_Ambient_NonLoop
 concurrencyOverrides {resolutionRule=PreventNew}
}
